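# Base editor utility widget for plotting graph point data into a render target, with empty draw, clear and material hooks apparently meant for subclasses.
# SetGraphData sorts points by X through inline Python and computes bounds.
# Drawing resolves auto-size or fixed display ranges, filters points by visible X, then creates or resizes the render target and schedules a next-tick callback.
# Data changes fire a dispatcher; range changes restart a short debounce timer.
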EditorUtilityWidgetBlueprint /Game/NeverCook/EditorUtilities/Generic/GraphWidgets/GraphRenderer/EUW_GraphRenderer
graphs [
 CleanAndSortGraphData {nodes [
   CleanAndSortGraphData() {is=title K2Node_FunctionEntry}
   n1 = Get CurrentGraphData() -> (CurrentGraphData:Array of Graph Renderer Point Data Structures)
   > Branch(Condition="integer > integer"(A=Length(TargetArray:Array of Graph Renderer Point Data Structures=@n1), B=1))
   n2 = Get CurrentGraphData()
   n3 = BreakVector2D(InVec=Break Graph Renderer Point Data!(GraphRendererPointData:"Graph Renderer Point Data Structure (by ref)"="Get (a copy)"(Array:Array of Graph Renderer Point Data Structures=@n2) -> (Output:Graph Renderer Point Data Structure)) -> (Point_7_#1:Vector 2D Structure) {is=title K2Node_BreakStruct})
   > Set!(execute, CurrentMinMaxX_X:double=@n3, CurrentMinMaxX_Y:double=@n3) -> (then, Output_Get:Vector 2D Structure)
   > Set!(execute, CurrentMinMaxY_X:double=@n3.Y, CurrentMinMaxY_Y:double=@n3.Y) -> (then, Output_Get)
   > n4 = For Each Loop(Array:Array of Graph Renderer Point Data Structures=Get CurrentGraphData()) -> (Array Element:Graph Renderer Point Data Structure)
   n5 = Sequence!(execute=@n4.Completed) -> (then_0, then_1, then_2:Exec)
   > n6 = Execute Python Script!(
    execute
    PythonScript="sorted_map = sorted(in_data_map.items())"
    in_data_map:"Map of Float (double-precision)s to Integers"=Get l_xToIndexMap() -> (l_xToIndexMap:"Map of Float (double-precision)s to Integers")
   ) -> (then, ReturnValue, sorted_map:"Map of Float (double-precision)s to Integers") {inputs [in_data_map], outputs [sorted_map]}
   > n7 = Set l_xToIndexMap(l_xToIndexMap:"Map of Float (double-precision)s to Integers"=@n6.sorted_map) -> (Output_Get:"Map of Float (double-precision)s to Integers")
   > n8 = Values!(execute, TargetMap:"Map of Float (double-precision)s to Integers"=@n7) -> (then, Values:Array of Integers)
   > n9 = For Each Loop(Array:Array of Integers=@n8) -> (Array Element:int)
   > Set Array Elem!(
    execute
    TargetArray:Array of Graph Renderer Point Data Structures=Get l_sortedData() -> (l_sortedData:Array of Graph Renderer Point Data Structures)
    Index:int=@n9.Array Index
    Item:"Graph Renderer Point Data Structure (by ref)"="Get (a copy)"(Array=Get CurrentGraphData(), Dimension 1=@n9)
    bSizeToFit:bool=true
   ) -> (then)
   Set CurrentGraphData(execute=@n5.then_2, CurrentGraphData:Array of Graph Renderer Point Data Structures=Get l_sortedData()) -> (Output_Get:Array of Graph Renderer Point Data Structures)
   Set l_point(execute=@n4.LoopBody, l_point:Vector 2D Structure=Break Graph Renderer Point Data(GraphRendererPointData=@n4)) -> (Output_Get:Vector 2D Structure)
   > Add!(
    execute
    TargetMap:"Map of Float (double-precision)s to Integers"=Get l_xToIndexMap()
    Key:"Float (double-precision) (by ref)"=BreakVector2D(InVec=Get l_point() -> (l_point:Vector 2D Structure))
    Value:"Integer (by ref)"=@n4.Array Index
   ) -> (then) {is=title K2Node_CallFunction}
   n10 = BreakVector2D(InVec=Get l_point())
   n11 = BreakVector2D(InVec=Get CurrentMinMaxX() -> (CurrentMinMaxX:Vector 2D Structure))
   > Set CurrentMinMaxX(CurrentMinMaxX:Vector 2D Structure=MakeVector2D(
     X="Min (Float)"(A:double=@n10, B:double=@n11) -> (ReturnValue:double) {is=title K2Node_CommutativeAssociativeBinaryOperator}
     Y="Max (Float)"(A=@n10, B=@n11.Y)
    )) -> (Output_Get:Vector 2D Structure)
   n12 = BreakVector2D(InVec=Get CurrentMinMaxY() -> (CurrentMinMaxY:Vector 2D Structure))
   > Set CurrentMinMaxY(CurrentMinMaxY:Vector 2D Structure=MakeVector2D(X="Min (Float)"(A=@n10.Y, B=@n12), Y="Max (Float)"(A=@n10.Y, B=@n12.Y))) -> (Output_Get:Vector 2D Structure)
  ]}
 ClearGraph {nodes [ClearGraph() {is=title K2Node_FunctionEntry}]}
 DispatchEvent_MinMaxUpdated {nodes [
   DispatchEvent_MinMaxUpdated() {is=title K2Node_FunctionEntry}
   > Call On Min Max Updated!(execute:Exec, self:EUW Graph Renderer Object Reference, GraphRenderer:EUW Graph Renderer Object Reference=Self-Reference()) -> (then:Exec) {is=title K2Node_CallDelegate}
  ]}
 DrawGraphData {nodes [
   DrawGraphData() {is=title K2Node_FunctionEntry}
   > ValidateDisplaySizes()
   > FilterGraphDataByZoomLevel()
   > RT_ValidateThenDraw()
  ]}
 EventGraph {nodes [
   Event Construct()
   > Set DisplayMinMaxX(DisplayMinMaxX:Vector 2D Structure=Get DefaultDisplayMinMaxX() -> (DefaultDisplayMinMaxX:Vector 2D Structure)) -> (Output_Get:Vector 2D Structure)
   > Set DisplayMinMaxY(DisplayMinMaxY:Vector 2D Structure=Get DefaultDisplayMinMaxY() -> (DefaultDisplayMinMaxY:Vector 2D Structure)) -> (Output_Get:Vector 2D Structure)
  ]}
 FilterGraphDataByZoomLevel {nodes [
   FilterGraphDataByZoomLevel() {is=title K2Node_FunctionEntry}
   n1 = Get DisplayMinMaxX() -> (DisplayMinMaxX:Vector 2D Structure)
   n2 = Get CurrentMinMaxX()
   > n3 = Branch(Condition=OR Boolean(
     A=Get AutoSizeX() -> (AutoSizeX:bool)
     B="Equal (Vector2D)"!(A:Vector 2D Structure=@n2, B:Vector 2D Structure=@n1, ErrorTolerance:float=0.001) -> (ReturnValue:bool=~) {is=title K2Node_PromotableOperator}
    ))
   > Set FilteredGraphDataToDraw(FilteredGraphDataToDraw:Array of Graph Renderer Point Data Structures=Get CurrentGraphData()) -> (Output_Get:Array of Graph Renderer Point Data Structures)
   n4 = Sequence(execute=@n3.else)
   > n5 = For Each Loop(Array:Array of Graph Renderer Point Data Structures=Get CurrentGraphData()) -> (Array Element:Graph Renderer Point Data Structure)
   Set FilteredGraphDataToDraw(execute=@n5.Completed, FilteredGraphDataToDraw=Get l_finalData() -> (l_finalData:Array of Graph Renderer Point Data Structures))
   Set l_curPoint(execute=@n5.LoopBody, l_curPoint:Graph Renderer Point Data Structure=@n5) -> (Output_Get:Graph Renderer Point Data Structure)
   n6 = BreakVector2D(InVec=Break Graph Renderer Point Data!(GraphRendererPointData=Get l_curPoint() -> (l_curPoint:Graph Renderer Point Data Structure)) -> (Point_7_#1, Severity_12_#2:double=~))
   n7 = BreakVector2D(InVec=Get DisplayMinMaxX())
   > Branch(Condition=AND Boolean(A="float >= float"(A=@n6, B=@n7), B="float <= float"(A=@n6, B=@n7.Y) -> (ReturnValue=~)))
   > Add(
    TargetArray:Array of Graph Renderer Point Data Structures=Get l_finalData()
    NewItem:"Graph Renderer Point Data Structure (by ref)"=Get l_curPoint()
   ) {is=title K2Node_CallArrayFunction}
   n8 = BreakVector2D(InVec=Get l_minMaxY() -> (l_minMaxY:Vector 2D Structure))
   n9 = BreakVector2D(InVec=Break Graph Renderer Point Data(GraphRendererPointData=Get l_curPoint()) -> (Severity_12_#2=~))
   > Set l_minMaxY(l_minMaxY:Vector 2D Structure=MakeVector2D(X="Min (Float)"(A=@n9.Y, B=@n8), Y="Max (Float)"(A=@n9.Y, B=@n8.Y))) -> (Output_Get:Vector 2D Structure)
   n10 = Switch on Enum_GraphRenderAutosize!(execute=@n4.then_2, Selection:Enum_GraphRenderAutosize Enum=Get AutoSizeY() -> (AutoSizeY:Enum_GraphRenderAutosize Enum=NewEnumerator0)) -> (NewEnumerator0:Exec, NewEnumerator1:Exec, NewEnumerator2:Exec) {is=title K2Node_SwitchEnum}
   Set DisplayMinMaxY(execute=@n10.NewEnumerator1, DisplayMinMaxY=Get l_minMaxY())
  ]}
 GetCurrentMinMaxX {nodes [
   GetCurrentMinMaxX() {is=title K2Node_FunctionEntry}
   > Return Node!(execute, CurrentMinMaxX:Vector 2D Structure=Get CurrentMinMaxX())
  ]}
 GetCurrentMinMaxY {nodes [
   GetCurrentMinMaxY() {is=title K2Node_FunctionEntry}
   > Return Node!(execute, CurrentMinMaxY:Vector 2D Structure=Get CurrentMinMaxY())
  ]}
 GetDrawColorFromIndex {nodes [
   n1 = GetDrawColorFromIndex!() -> (then, index:int) {is=title K2Node_FunctionEntry}
   n2 = BreakVector2D(InVec=Get CurrentMinMaxY())
   n3 = "Get (a copy)"(Array=Get FilteredGraphDataToDraw() -> (FilteredGraphDataToDraw:Array of Graph Renderer Point Data Structures), Dimension 1=@n1)
   n4 = NormalizeToRange!(
    Value:double=BreakVector2D(InVec=Break Graph Renderer Point Data(GraphRendererPointData=@n3) -> (Severity_12_#2=~)).Y
    RangeMin:double=@n2
    RangeMax:double=@n2.Y
   ) -> (ReturnValue:double)
   > Return Node!(
    execute
    ReturnValue:Linear Color Structure="Lerp (LinearColor)"(A=(R=0.1442,G=0.5469,B=0.1339,A=1), B=(R=0.6562,G=0.1434,B=0.1094,A=1), Alpha="Clamp (Float)"(Value=@n4, Min=0, Max=1))
   )
  ]}
 GetGraphValueFromPercentage {nodes [
   GetGraphValueFromPercentage!() -> (then, PercentX:double=~, FromTotal:bool=~) {is=title K2Node_FunctionEntry}
   > Set l_data(l_data:Array of Graph Renderer Point Data Structures=Get FilteredGraphDataToDraw()) -> (Output_Get:Array of Graph Renderer Point Data Structures)
   > n1 = Sequence!(execute) -> (then_0, then_1)
   > Set l_minMaxX(l_minMaxX:Vector 2D Structure=Get DisplayMinMaxX()) -> (Output_Get:Vector 2D Structure)
   n2 = Get DisplayMinMaxY() -> (DisplayMinMaxY:Vector 2D Structure)
   > Set l_minMaxY(l_minMaxY=@n2)
   n3 = float * float(
    A=Get PercentX() -> (PercentX:double)
    B:int=Last Index!(TargetArray:Array of Graph Renderer Point Data Structures=Get l_data() -> (l_data:Array of Graph Renderer Point Data Structures)) -> (ReturnValue=0)
   ) {is=title K2Node_PromotableOperator}
   n4 = Break Graph Renderer Point Data!(GraphRendererPointData="Get (a copy)"(Array=Get l_data(), Dimension 1=Floor(A=@n3) {is=title K2Node_CallFunction})) -> (
    Point_7_#1_X:double
    Point_7_#1_Y:double
    Severity_12_#2=~
   )
   n5 = BreakVector2D(InVec=Get DisplayMinMaxY())
   n6 = BreakVector2D(InVec=Get DisplayMinMaxX())
   Return Node!(
    execute=@n1.then_1
    Point:Vector 2D Structure=MakeVector2D(X=@n4, Y=@n4.Point_7_#1_Y)
    PercentValues:Vector 2D Structure=MakeVector2D(
     X=NormalizeToRange(Value=@n4, RangeMin=@n6, RangeMax=@n6.Y)
     Y=NormalizeToRange(Value=@n4.Point_7_#1_Y, RangeMin=@n5, RangeMax=@n5.Y)
    )
   )
   Select!(
    NewEnumerator0:Vector 2D Structure
    NewEnumerator1:Vector 2D Structure
    NewEnumerator2:Vector 2D Structure=@n2
    Index:Enum_GraphRenderAutosize Enum=Get AutoSizeY() -> (AutoSizeY=NewEnumerator0)
   ) -> (ReturnValue:Vector 2D Structure)
  ]}
 GetWidgetMaterial {nodes [
   GetWidgetMaterial() {is=title K2Node_FunctionEntry}
   > Return Node!(execute, ReturnValue:Material Instance Dynamic Object Reference)
  ]}
 InvalidateMinMax {nodes [
   InvalidateMinMax() {is=title K2Node_FunctionEntry}
   > Clear and Invalidate Timer by Handle!(execute, WorldContextObject:Object Reference, Handle=Get timerHandle_MinMaxUpdate() -> (timerHandle_MinMaxUpdate:Timer Handle Structure)) -> (then)
   > n1 = Set Timer by Event(Delegate=Create Event(), Time=0.01, bLooping=false, bMaxOncePerFrame=true)
   > Set timerHandle_MinMaxUpdate(timerHandle_MinMaxUpdate:Timer Handle Structure=@n1) -> (Output_Get:Timer Handle Structure)
  ]}
 OnDataUpdated {nodes [OnDataUpdated!() -> (then, GraphRenderer:EUW Graph Renderer Object Reference) {is=title K2Node_FunctionEntry}]}
 OnMinMaxUpdated {nodes [OnMinMaxUpdated!() -> (then, GraphRenderer:EUW Graph Renderer Object Reference) {is=title K2Node_FunctionEntry}]}
 RT_Draw {nodes [RT_Draw() {is=title K2Node_FunctionEntry}]}
 RT_ValidateThenDraw {nodes [
   RT_ValidateThenDraw() {is=title K2Node_FunctionEntry}
   > n1 = Sequence()
   n2 = GetLocalSize!(Geometry:"Geometry Structure (by ref)"=GetCachedGeometry!(self:Widget Object Reference) -> (ReturnValue:"Geometry Structure (by ref)")) -> (ReturnValue_X:double, ReturnValue_Y:double)
   n3 = Truncate(A=@n2)
   n4 = Truncate(A=@n2.ReturnValue_Y)
   > Set!(execute, CurrentRTSize_X:double="To Float (Integer)"(InInt=@n3), CurrentRTSize_Y:double="To Float (Integer)"(InInt=@n4)) -> (then, Output_Get)
   > n5 = Is Valid(InputObject=Get CurrentRT() -> (CurrentRT:Texture Render Target 2D Object Reference))
   n6 = CreateRenderTarget2D!(
    execute=@n5.Is Not Valid
    WorldContextObject:Object Reference
    Width:int=@n3
    Height:int=@n4
    Format:ETextureRenderTargetFormat Enum=RTF_RGBA8
    ClearColor:Linear Color Structure=(R=0,G=0,B=0,A=0)
    bAutoGenerateMipMaps:bool=false
    bSupportUAVs:bool=false
   ) -> (then, ReturnValue:Texture Render Target 2D Object Reference)
   > n7 = Set CurrentRT(CurrentRT:Texture Render Target 2D Object Reference=@n6) -> (Output_Get:Texture Render Target 2D Object Reference)
   > n8 = GetWidgetMaterial!(execute, self) -> (then, ReturnValue:Material Instance Dynamic Object Reference) {is=title K2Node_CallFunction}
   > SetTextureParameterValue(self=@n8, ParameterName=Texture, Value=@n7)
   n9 = Get!() -> (CurrentRTSize_X:double, CurrentRTSize_Y:double)
   ResizeRenderTarget2D!(
    execute=@n5.Is Valid
    TextureRenderTarget:Texture Render Target 2D Object Reference=Get CurrentRT()
    Width:int=Truncate(A=@n9)
    Height:int=Truncate(A=@n9.CurrentRTSize_Y)
   ) -> (then)
   > ClearRenderTarget2D!(
    execute
    WorldContextObject:Object Reference
    TextureRenderTarget:Texture Render Target 2D Object Reference=Get CurrentRT()
    ClearColor:Linear Color Structure=(R=0,G=0,B=0,A=0)
   ) -> (then)
   Set Timer for Next Tick by Event!(execute=@n1.then_1, Delegate:Delegate=Create Event()) -> (then, ReturnValue:Timer Handle Structure) {is=title K2Node_CallFunction}
  ]}
 SetAutoSizeY {nodes [
   n1 = SetAutoSizeY!() -> (then, NewAutoSizeY:Enum_GraphRenderAutosize Enum=NewEnumerator0) {is=title K2Node_FunctionEntry}
   > Branch(Condition="Not Equal (Enum)"!(A:Enum_GraphRenderAutosize Enum=@n1, B:Enum_GraphRenderAutosize Enum=Get AutoSizeY() -> (AutoSizeY=NewEnumerator0)) -> (ReturnValue:bool))
   > Set AutoSizeY(AutoSizeY:Enum_GraphRenderAutosize Enum=@n1) -> (Output_Get:Enum_GraphRenderAutosize Enum=NewEnumerator0)
   > DrawGraphData() {is=title K2Node_CallFunction}
  ]}
 SetDisplayArea {nodes [
   SetDisplayArea!() -> (
    then
    NewAutoSizeX:bool=true
    NewDisplayMinMaxX:Vector 2D Structure
    NewAutoSizeY:Enum_GraphRenderAutosize Enum=NewEnumerator0
    NewDisplayMinMaxY:Vector 2D Structure
   ) {is=title K2Node_FunctionEntry}
   > n1 = Sequence!(execute) -> (then_0, then_1, then_2:Exec, then_3:Exec)
   n2 = Get NewAutoSizeY() -> (NewAutoSizeY:Enum_GraphRenderAutosize Enum=NewEnumerator0)
   n3 = Get AutoSizeY() -> (AutoSizeY=NewEnumerator0)
   n4 = Get AutoSizeX()
   n5 = Get NewAutoSizeX() -> (NewAutoSizeX:bool)
   > Branch(Condition=OR Boolean(
     A="Not Equal (Boolean)"(A=@n5, B=@n4) -> (ReturnValue=~)
     B="Equal (Enum)"(A:Enum_GraphRenderAutosize Enum=@n2, B:Enum_GraphRenderAutosize Enum=@n3)
    ))
   > Set l_changed(l_changed:bool=true) -> (Output_Get:bool)
   > Set AutoSizeX(AutoSizeX:bool=Get AutoSizeX()) -> (Output_Get:bool)
   > Set AutoSizeY(AutoSizeY=Get NewAutoSizeY() -> (NewAutoSizeY=NewEnumerator0)) -> (Output_Get=NewEnumerator0)
   Branch(
    execute=@n1.then_1
    Condition="Not Equal (Vector2D)"!(
     A:Vector 2D Structure=Get NewDisplayMinMaxX() -> (NewDisplayMinMaxX:Vector 2D Structure)
     B:Vector 2D Structure=Get DisplayMinMaxX()
     ErrorTolerance:float
    ) -> (ReturnValue:bool=~) {is=title K2Node_PromotableOperator}
   )
   > SetDisplayMinMax!(execute, self, DisplayMinMax:Vector 2D Structure=Get NewDisplayMinMaxX(), IsX:bool=true) -> (then)
   > n6 = Branch(Condition=Get AutoSizeX())
   Set l_changed(execute=@n6.else, l_changed=true)
   Branch(
    execute=@n1.then_2
    Condition="Not Equal (Vector2D)"(A=Get NewDisplayMinMaxY() -> (NewDisplayMinMaxY:Vector 2D Structure), B=Get DisplayMinMaxY()) -> (ReturnValue=~)
   )
   > SetDisplayMinMax(DisplayMinMax=Get NewDisplayMinMaxY(), IsX=false)
   > n7 = Switch on Enum_GraphRenderAutosize(Selection=Get NewAutoSizeY() -> (NewAutoSizeY=NewEnumerator0))
   Set l_changed(execute=@n7.NewEnumerator2, l_changed=true)
   Branch(execute=@n1.then_3, Condition=Get l_changed() -> (l_changed:bool))
   > DrawGraphData()
  ]}
 SetDisplayMinMax {nodes [
   n1 = SetDisplayMinMax!() -> (then, DisplayMinMax:Vector 2D Structure, IsX:bool=~) {is=title K2Node_FunctionEntry}
   > Set l_new(l_new:Vector 2D Structure=@n1) -> (Output_Get:Vector 2D Structure)
   n2 = Get DisplayMinMaxX()
   n3 = Get DisplayMinMaxY()
   n4 = Get IsX() -> (IsX:bool)
   > Branch(Condition="Not Equal (Vector2D)"(
     A=Get l_new() -> (l_new:Vector 2D Structure)
     B=Select!(Option 0:Vector 2D Structure=@n3, Option 1:Vector 2D Structure=@n2, Index:bool=@n4) -> (ReturnValue)
    ) -> (ReturnValue=~))
   > n5 = Branch(Condition=Get IsX())
   n6 = Get l_new()
   > n7 = Set DisplayMinMaxX(DisplayMinMaxX=@n6)
   InvalidateMinMax(execute=[@n7.then, @n8.then]) {is=title K2Node_CallFunction}
   n8 = Set DisplayMinMaxY(execute=@n5.else, DisplayMinMaxY=@n6)
  ]}
 SetGraphData {nodes [
   n1 = SetGraphData!() -> (then, GraphData:Array of Graph Renderer Point Data Structures) {is=title K2Node_FunctionEntry}
   > Set CurrentGraphData(CurrentGraphData=@n1)
   > UpdateGraph()
   > Call On Data Updated!(execute:Exec, self:EUW Graph Renderer Object Reference, GraphRenderer:EUW Graph Renderer Object Reference=Self-Reference()) -> (then:Exec) {is=title K2Node_CallDelegate}
  ]}
 UpdateGraph {nodes [
   UpdateGraph() {is=title K2Node_FunctionEntry}
   > n1 = Sequence!(execute) -> (then_0, then_1, then_2)
   > ClearGraph() {is=title K2Node_CallFunction}
   CleanAndSortGraphData(execute=@n1.then_1) {is=title K2Node_CallFunction}
   DrawGraphData(execute=@n1.then_2)
  ]}
 ValidateDisplaySizes {nodes [
   ValidateDisplaySizes() {is=title K2Node_FunctionEntry}
   > n1 = Sequence()
   > n2 = Branch(Condition=Get AutoSizeX())
   > SetDisplayMinMax(DisplayMinMax=Get CurrentMinMaxX(), IsX=true) {is=title K2Node_CallFunction}
   SetDisplayMinMax(execute=@n2.else, DisplayMinMax=Get DisplayMinMaxX(), IsX=true)
   n3 = Switch on Enum_GraphRenderAutosize(execute=@n1.then_1, Selection=Get AutoSizeY() -> (AutoSizeY=NewEnumerator0))
   SetDisplayMinMax(execute=[@n3.NewEnumerator0, @n3.NewEnumerator1], DisplayMinMax=Get CurrentMinMaxY(), IsX=false)
   SetDisplayMinMax(execute=@n3.NewEnumerator2, DisplayMinMax=Get DisplayMinMaxY(), IsX=false)
  ]}
]
variables [
 AutoSizeX {type=boolean, default=true}
 AutoSizeY {type=Enum_GraphRenderAutosize, default=Fit Data To Render Area}
 CurrentGraphData {type="array<GraphRendererPointData>", default []}
 CurrentMinMaxX {type=Vector2D, default=<0 0>}
 CurrentMinMaxY {type=Vector2D, default=<0 0>}
 CurrentRT {type=TextureRenderTarget2D, default=null}
 CurrentRTSize {type=Vector2D, default=<0 0>}
 DefaultDisplayMinMaxX {type=Vector2D, default=<0 100>}
 DefaultDisplayMinMaxY {type=Vector2D, default=<0 100>}
 DisplayMinMaxX {type=Vector2D, default=<0 100>}
 DisplayMinMaxY {type=Vector2D, default=<0 100>}
 FilteredGraphDataToDraw {type="array<GraphRendererPointData>", default []}
 LastDrawSize {type=Vector2D, default=<0 0>}
 OnDataUpdated {default=()}
 OnMinMaxUpdated {default=()}
 timerHandle_MinMaxUpdate {type=TimerHandle}
]
widgets {info {rootWidgetClass=null, widgetCount=0}, widgets []}
