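# Editor utility widget, child of EUW_GraphRenderer, that draws graph data as a line graph on a render target.
# RT_Draw_Line scales filtered points to the display range and draws colored canvas segments, optional point boxes and a bottom line.
# It skips points landing on the previous X pixel and caches drawn ones in PixelPointMap.
# GetGraphValueFromPercentage searches that map by horizontal percentage; GetWidgetMaterial exposes the image's dynamic material.
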EditorUtilityWidgetBlueprint /Game/NeverCook/EditorUtilities/Generic/GraphWidgets/GraphRenderer/EUW_LineGraphRenderer
graphs [
 EventGraph {nodes [
   Event PreConstruct()
   Event Construct()
   Event Tick()
  ]}
 GetDrawLocationFromDataIndex {nodes [
   n1 = GetDrawLocationFromDataIndex!() -> (then, Index:int) {is=title K2Node_FunctionEntry}
   n2 = Get FilteredGraphDataToDraw() -> (FilteredGraphDataToDraw:Array of Graph Renderer Point Data Structures)
   > Set l_dataPoint(l_dataPoint:Vector 2D Structure=Break Graph Renderer Point Data!(GraphRendererPointData:"Graph Renderer Point Data Structure (by ref)"="Get (a ref)"(Array:Array of Graph Renderer Point Data Structures=@n2, Dimension 1=@n1) -> (Output:"Graph Renderer Point Data Structure (by ref)")) -> (Point_7_#1:Vector 2D Structure, Severity_12_#2:double=~) {is=title K2Node_BreakStruct}) -> (Output_Get:Vector 2D Structure)
   n3 = BreakVector2D(InVec=Get CurrentRTSize() -> (CurrentRTSize:Vector 2D Structure))
   n4 = BreakVector2D(InVec=Get l_dataPoint() -> (l_dataPoint:Vector 2D Structure))
   n5 = BreakVector2D(InVec=Get DisplayMinMaxX() -> (DisplayMinMaxX:Vector 2D Structure))
   n6 = Round(A=float * float(A=NormalizeToRange!(Value:double=@n4, RangeMin:double=@n5, RangeMax:double=@n5.Y) -> (ReturnValue:double), B=@n3) {is=title K2Node_PromotableOperator}) {is=title K2Node_CallFunction}
   n7 = BreakVector2D(InVec=Get DisplayMinMaxY() -> (DisplayMinMaxY:Vector 2D Structure))
   n8 = float - float(A=1, B=NormalizeToRange(Value=@n4.Y, RangeMin=@n7, RangeMax=@n7.Y))
   > Return Node!(execute, DrawLocation:Vector 2D Structure=MakeVector2D(X="To Float (Integer)"(InInt=@n6), Y=float * float(A=@n3.Y, B=@n8)), XGrid:int=@n6)
  ]}
 GetGraphValueFromPercentage {nodes [
   GetGraphValueFromPercentage!() -> (then, PercentX:double, FromTotal:bool) {is=title K2Node_FunctionEntry}
   > n1 = Sequence!(execute) -> (then_0, then_1, then_2:Exec)
   n2 = GetLocalSize!(Geometry:"Geometry Structure (by ref)"=GetCachedGeometry!(self:Widget Object Reference) -> (ReturnValue:"Geometry Structure (by ref)")) -> (ReturnValue_X:double, ReturnValue_Y:double)
   n3 = Get PercentX() -> (PercentX:double)
   > Set l_pixel(l_pixel:int=Round(A=float * float(A=@n3, B=@n2))) -> (Output_Get:int=0)
   n4 = Values!(
    execute=@n1.then_1
    TargetMap:Map of Integers to Vector 2D Structures=Get PixelPointMap() -> (PixelPointMap:Map of Integers to Vector 2D Structures)
   ) -> (then, Values:Array of Vector 2D Structures)
   > Set l_points(l_points:Array of Vector 2D Structures=@n4) -> (Output_Get:Array of Vector 2D Structures)
   > n5 = Keys!(execute, TargetMap:Map of Integers to Vector 2D Structures=Get PixelPointMap()) -> (then, Keys:Array of Integers)
   > n6 = For Each Loop(Array:Array of Integers=@n5) -> (Array Element:int)
   > Set l_Xvalue(l_Xvalue:int=@n6) -> (Output_Get:int=0)
   > Set l_index(l_index:int=@n6.Array Index) -> (Output_Get:int=0)
   > Branch(Condition="integer >= integer"(A=Get l_Xvalue() -> (l_Xvalue:int=0), B=Get l_pixel() -> (l_pixel:int=0)))
   > n7 = Branch(Condition="integer < integer"(A=Get l_index() -> (l_index:int=0), B=1))
   > Set l_index(l_index=0) -> (Output_Get=0)
   n8 = Branch(execute=@n7.else, Condition="integer > integer"(A=Get l_Xvalue() -> (l_Xvalue=0), B=Get l_pixel() -> (l_pixel=0)))
   > Set l_index(l_index=int - int(A=Get l_index() -> (l_index=0))) -> (Output_Get=0)
   Set l_index(execute=@n8.else, l_index=Get l_index() -> (l_index=0)) -> (Output_Get=0)
   n9 = BreakVector2D(InVec=Get DisplayMinMaxY())
   n10 = "Get (a copy)"(Array:Array of Vector 2D Structures=Get l_points() -> (l_points:Array of Vector 2D Structures), Dimension 1=Get l_index() -> (l_index=0)) -> (Output:Vector 2D Structure)
   n11 = BreakVector2D(InVec=@n10)
   n12 = BreakVector2D(InVec=Get DisplayMinMaxX())
   Return Node!(
    execute=@n1.then_2
    Point:Vector 2D Structure=@n10
    PercentValues:Vector 2D Structure=MakeVector2D(X=NormalizeToRange(Value=@n11, RangeMin=@n12, RangeMax=@n12.Y), Y=NormalizeToRange(Value=@n11.Y, RangeMin=@n9, RangeMax=@n9.Y))
   )
  ]}
 GetWidgetMaterial {nodes [
   GetWidgetMaterial() {is=title K2Node_FunctionEntry}
   > n1 = GetDynamicMaterial(self=Get ImageRT() -> (ImageRT:Image Object Reference))
   > Return Node!(execute, ReturnValue:Material Instance Dynamic Object Reference=@n1)
  ]}
 RT_Draw {nodes [
   RT_Draw() {is=title K2Node_FunctionEntry}
   > Parent: RT_Draw() {is=title K2Node_CallParentFunction}
   > RT_Draw_Line()
  ]}
 RT_Draw_Line {nodes [
   RT_Draw_Line() {is=title K2Node_FunctionEntry}
   n1 = Get CurrentRT() -> (CurrentRT:Texture Render Target 2D Object Reference)
   > Is Valid(InputObject=@n1)
   > n2 = BeginDrawCanvasToRenderTarget!(execute, WorldContextObject:Object Reference, TextureRenderTarget:Texture Render Target 2D Object Reference=@n1) -> (then, Canvas:Canvas Object Reference, Size:Vector 2D Structure, Context:Draw To Render Target Context Structure)
   > Set l_canvas(l_canvas:Canvas Object Reference=@n2) -> (Output_Get:Canvas Object Reference)
   > n3 = Sequence!(execute) -> (then_0, then_1)
   n4 = Length(TargetArray:Array of Graph Renderer Point Data Structures=Get CurrentGraphData() -> (CurrentGraphData:Array of Graph Renderer Point Data Structures))
   > Branch(Condition="integer > integer"(A=@n4, B=1) {is=title K2Node_CallFunction})
   > Set l_count(l_count:int=int - int(A=@n4, B=1) {is=title K2Node_CallFunction}) -> (Output_Get:int=0)
   > Clear!(execute, TargetMap:Map of Integers to Vector 2D Structures=Get PixelPointMap()) -> (then) {is=title K2Node_CallFunction}
   > n5 = Sequence!(execute) -> (then_0, then_1, then_2:Exec)
   n6 = "Get (a copy)"!(Array:Array of Graph Renderer Point Data Structures=Get FilteredGraphDataToDraw(), Dimension 1) -> (Output_Point_7_#1:Vector 2D Structure, Output_Severity_12_#2:double=~)
   > Set l_lastGridX(l_lastGridX:int=Round(A=BreakVector2D(InVec=@n6))) -> (Output_Get:int=0)
   > n7 = For Each Loop(Array:Array of Graph Renderer Point Data Structures=Get FilteredGraphDataToDraw()) -> (Array Element:Graph Renderer Point Data Structure)
   > Set l_curIndex(l_curIndex:int=@n7.Array Index) -> (Output_Get:int=0)
   > Branch(Condition="integer > integer"(A=Get l_curIndex() -> (l_curIndex:int=0), B=0))
   > n8 = GetDrawLocationFromDataIndex!(execute, self, Index:int=Get l_curIndex() -> (l_curIndex=0)) -> (then, DrawLocation:Vector 2D Structure, XGrid:int=0) {is=title K2Node_CallFunction}
   > Set l_endPoint(l_endPoint:Vector 2D Structure=@n8) -> (Output_Get:Vector 2D Structure)
   > Branch(Condition="Not Equal (Integer)"(A=@n8.XGrid, B=Get l_lastGridX() -> (l_lastGridX:int=0)))
   > Set l_lastGridX(l_lastGridX=@n8.XGrid) -> (Output_Get=0)
   > n9 = GetDrawColorFromIndex!(execute, self:EUW Graph Renderer Object Reference, index:int=Get l_curIndex() -> (l_curIndex=0)) -> (then, ReturnValue:Linear Color Structure)
   > Set l_drawColor(l_drawColor:Linear Color Structure=@n9) -> (Output_Get:Linear Color Structure)
   > n10 = GetDrawLocationFromDataIndex(Index=int - int(A=Get l_curIndex() -> (l_curIndex=0), B=1)) -> (XGrid=0)
   > Draw Line!(
    execute
    self:Canvas Object Reference=Get l_canvas() -> (l_canvas:Canvas Object Reference)
    ScreenPositionA:Vector 2D Structure=@n10
    ScreenPositionB:Vector 2D Structure=Get l_endPoint() -> (l_endPoint:Vector 2D Structure)
    Thickness:float=1
    RenderColor:Linear Color Structure=Get l_drawColor() -> (l_drawColor:Linear Color Structure)
   ) -> (then)
   > Add!(
    execute
    TargetMap:Map of Vector 2D Structures to Integers=Get l_drawnPoints() -> (l_drawnPoints:Map of Vector 2D Structures to Integers)
    Key:"Vector 2D Structure (by ref)"=Get l_endPoint()
    Value:"Integer (by ref)"=Get l_curIndex() -> (l_curIndex=0)
   ) -> (then) {is=title K2Node_CallFunction}
   > Add(
    TargetMap:Map of Integers to Vector 2D Structures=Get PixelPointMap()
    Key:"Integer (by ref)"=Get l_lastGridX() -> (l_lastGridX=0)
    Value:"Vector 2D Structure (by ref)"=Get l_endPoint()
   )
   Branch(execute=@n5.then_1, Condition=Get DrawPoints() -> (DrawPoints:bool))
   n11 = Get l_drawnPoints()
   > n12 = Values(TargetMap:Map of Vector 2D Structures to Integers=@n11) -> (Values:Array of Integers)
   > n13 = Keys(TargetMap:Map of Vector 2D Structures to Integers=@n11) -> (Keys:Array of Vector 2D Structures)
   > n14 = For Each Loop(Array:Array of Vector 2D Structures=@n13) -> (Array Element:Vector 2D Structure)
   > n15 = GetDrawColorFromIndex(index="Get (a copy)"!(Array:Array of Integers=@n12, Dimension 1=@n14.Array Index) -> (Output:int))
   > Set l_drawColor(l_drawColor=@n15)
   n16 = float / float(A=BreakVector2D(InVec=Get CurrentRTSize()), B:int=Length(TargetArray=Get FilteredGraphDataToDraw()))
   n17 = Get PointDrawThickness() -> (PointDrawThickness:float)
   > Draw Box!(
    execute
    self:Canvas Object Reference=Get l_canvas()
    ScreenPosition:Vector 2D Structure=vector2d - vector2d!(A:Vector 2D Structure=@n14, B:Vector 2D Structure=(X=0.5,Y=0.5)) -> (ReturnValue:Vector 2D Structure)
    ScreenSize:Vector 2D Structure=(X=1,Y=1)
    Thickness:float="Clamp (Float)"(Value=float * float(A=@n17, B=@n16), Min=2, Max=10)
    RenderColor:Linear Color Structure=Get l_drawColor()
   ) -> (then)
   n18 = Get!() -> (l_size_X:double, l_size_Y:double)
   Draw Line!(
    execute=@n5.then_2
    self=Get l_canvas()
    ScreenPositionA_X:double=0
    ScreenPositionA_Y:double=@n18.l_size_Y
    ScreenPositionB_X:double=@n18
    ScreenPositionB_Y:double=@n18.l_size_Y
    Thickness=1
    RenderColor=(R=1,G=1,B=1,A=1)
   ) -> (then)
   EndDrawCanvasToRenderTarget!(execute=@n3.then_1, WorldContextObject:Object Reference, Context:"Draw To Render Target Context Structure (by ref)"=@n2.Context) -> (then)
  ]}
]
variables [
 DrawPoints {type=boolean, default=true}
 PixelPointMap {default {}}
 PointDrawThickness {type=number, default=0.25}
]
widgets {
 info {
  parentClass=&/EUW_GraphRenderer.EUW_GraphRenderer_C
  rootWidgetClass=&Border
  widgetCount=2
 }
 widgets [
  Border_61 {widgetClassPath=&Border, widget=Border_61, properties {padding=<ltrb 0 0 0 0>}}
  ImageRT {
   widgetClassPath=&Image
   bIsVariable=true
   parent=Border_61
   widget=ImageRT
   properties {brush {
     bIsDynamicallyLoaded=false
     drawAs=Image
     imageType=FullColor
     mirroring=NoMirror
     resourceName=None
     resourceObject=/Game/NeverCook/EditorUtilities/Generic/GraphWidgets/GraphRenderer/Art/M_UI_LineGraph
     tiling=NoTile
     imageSize=<32 32>
     margin=<ltrb 0 0 0 0>
     outlineSettings {
      bUseBrushTransparency=false
      roundingType=HalfHeightRadius
      width=0
      color {colorUseRule=UseColor_Specified, specifiedColor=<rgba 0 0 0 0>}
      cornerRadii=<0 0 0 0>
     }
     tintColor {colorUseRule=UseColor_Specified, specifiedColor=<rgba 1 1 1 1>}
     uVRegion {bIsValid=false, max=<0 0>, min=<0 0>}
    }}
   slot {class=BorderSlot, properties {padding=<ltrb 0 0 0 0>}}
  }
 ]
}
